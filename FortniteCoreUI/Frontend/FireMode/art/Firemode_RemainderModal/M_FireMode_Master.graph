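# Translucent UI master material for Fire Mode modal art, showing an image parameter (default autofire texture) with an optional colored banner and a border.
# The image is scaled about its center, then tinted by an angled bottom-banner mask and finally by the border color.
# Border masks come from rotated, offset UV edges sharpened by multiply and saturate.
# Scalar lerp switches, rather than static switches, toggle the banner and choose full-edge, two-sided or four-offset borders.
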Material /FortniteCoreUI/Frontend/FireMode/art/Firemode_RemainderModal/M_FireMode_Master
graph {
 properties {
  bUsedWithStaticMesh=false
  blendMode=BLEND_Translucent
  materialDomain=MD_UI
  pixelDepthOffsetMode=PDOM_Legacy
  usedShadingModels=Default Lit
 }
 nodes [
  Abs(None=Add(A=TextureCoordinate()) {constB=-0.5})
  n1 = MaterialFunctionCall() {materialFunction=&GetUserInterfaceUV}
  n2 = MaterialFunctionCall(Float2=@n1.Pixel Size) {materialFunction=&BreakOutFloat2Components}
  n3 = Divide(A=@n2.R, B=@n2.R)
  n4 = AppendVector(A=@n3, B=Constant() {r=1})
  n5 = Multiply(A=OneMinus(None=@n3)) {constB=0.5}
  n6 = Constant()
  n7 = MaterialFunctionCall(
   UVs=Add(A=Multiply(A=@n1.Normalized UV, B=@n4), B=AppendVector(A=@n5, B=@n6))
   Texture Scale=ScalarParameter() {defaultValue=1, group=BottomBanner, parameterName=ImageScale, sliderMax=5}
  ) {materialFunction=&ScaleUVsByCenter}
  n8 = Rotator(
   Coordinate=Add(
    A=MaterialFunctionCall() {materialFunction=&GetUserInterfaceUV}."9-Slice UV (No Tiling)"
    B=ScalarParameter() {defaultValue=0.25, group=BottomBanner, parameterName=Offset}
   ) {constB=-0.5}
   Time=ScalarParameter() {defaultValue=0.52, group=BottomBanner, parameterName=BottomBanner_Angle}
  ) {speed=6.2832}
  n9 = Multiply(
   A=Saturate(None=ComponentMask(None=@n8) {g=true})
   B=ScalarParameter() {defaultValue=80, group=BottomBanner, parameterName=BottomBanner_sharpness}
  )
  n10 = Saturate(None=OneMinus(None=Saturate(None=@n9)))
  n11 = ScalarParameter() {group=BottomBanner, parameterName=Switch_BottomBanner, sliderMax=1, sortPriority=0}
  n12 = MaterialFunctionCall() {materialFunction=&BreakOutFloat2Components}
  n13 = ScalarParameter() {defaultValue=0.495, group=FullEdge_Or_TwoSided, parameterName=EdgeOuter}
  n14 = Subtract(A=@n12.R, B=@n13)
  n15 = ScalarParameter() {defaultValue=128, group=FullEdge_Or_TwoSided, parameterName=EdgeOuterFalloff}
  n16 = Subtract(A=@n12.G, B=@n13)
  n17 = Add(A=Saturate(None=Multiply(A=@n14, B=@n15)), B=Saturate(None=Multiply(A=@n15, B=@n16)))
  n18 = TextureCoordinate()
  n19 = ScalarParameter() {defaultValue=-0.03, group=FullEdge_Or_TwoSided, parameterName=TwoSideAngle}
  n20 = Abs(None=Add(A=Rotator(Coordinate=@n18, Time=@n19) {speed=6.2832}) {constB=-0.5})
  n21 = Multiply(
   A=Subtract(A=ComponentMask(None=@n20) {r=true}, B=ScalarParameter() {defaultValue=0.385, group=FullEdge_Or_TwoSided, parameterName=TwoSidedOffset}) {constB=0.25}
   B=ScalarParameter() {defaultValue=256, group=FullEdge_Or_TwoSided, parameterName=TwoSidedSharpness}
  ) {constB=64}
  n22 = LinearInterpolate(
   A=OneMinus(None=Saturate(None=@n17))
   B=Saturate(None=OneMinus(None=@n21))
   Alpha=ScalarParameter() {group=FullEdge_Or_TwoSided, parameterName=Switch_FullEdge_Or_TwoSided, sliderMax=1, sortPriority=0}
  )
  n23 = Divide(A=ScalarParameter() {group=Offset, parameterName=TopAngle}) {constB=360}
  n24 = MaterialFunctionCall() {materialFunction=&GetUserInterfaceUV}
  n25 = ComponentMask(None=Add(
    A=Rotator(Coordinate=@n24.Normalized UV, Time=@n23) {centerX=1, speed=6.2832}
    B=ScalarParameter() {defaultValue=-0.1, group=Offset, parameterName=TopOffset}
   )) {g=true}
  n26 = ScalarParameter() {defaultValue=128, group=Offset, parameterName=HorizontalSharpness}
  n27 = Rotator(Coordinate=@n24.Normalized UV, Time=Divide(A=ScalarParameter() {group=Offset, parameterName=BottomAngle}) {constB=360}) {centerX=0, speed=6.2832}
  n28 = ScalarParameter() {defaultValue=0.1, group=Offset, parameterName=BottomOffset}
  n29 = OneMinus(None=ComponentMask(None=Add(A=@n27, B=@n28)) {g=true})
  n30 = Multiply(A=Saturate(None=Multiply(A=@n25, B=@n26)), B=Saturate(None=Multiply(A=@n29, B=@n26)))
  n31 = Divide(A=ScalarParameter() {group=Offset, parameterName=LeftAngle}) {constB=360}
  n32 = ComponentMask(None=Add(
    A=Rotator(Coordinate=@n24.9-Slice UV, Time=@n31) {speed=6.2832}
    B=ScalarParameter() {defaultValue=-0.1, group=Offset, parameterName=LeftOffset}
   )) {r=true}
  n33 = ScalarParameter() {defaultValue=32, group=Offset, parameterName=VerticalSharpness}
  n34 = Rotator(Coordinate=@n24.9-Slice UV, Time=Divide(A=ScalarParameter() {group=Offset, parameterName=RightAngle}) {constB=360}) {centerX=0, speed=6.2832}
  n35 = ScalarParameter() {defaultValue=0.1, group=Offset, parameterName=RightOffset}
  n36 = OneMinus(None=ComponentMask(None=Add(A=@n34, B=@n35)) {r=true})
  n37 = Multiply(A=Saturate(None=Multiply(A=@n32, B=@n33)), B=Saturate(None=Multiply(A=@n36, B=@n33)))
  n38 = Saturate(None=Multiply(A=Saturate(None=@n30), B=Saturate(None=@n37)))
  n39 = ScalarParameter() {group=Offset, parameterName=Switch_to_4Offsets, sliderMax=1, sortPriority=0}
  LinearInterpolate(
   A=LinearInterpolate(
    A=TextureSampleParameter2D(UVs=@n7.UVs) {group=Item, parameterName=Image, texture=/FortniteCoreUI/Frontend/FireMode/art/T_Bacchus_Autofire}.RGB
    B=VectorParameter() {group=BottomBanner, parameterName=BottomBannerColor, defaultValue=<rgba 0 1 0.9647 1>}.RGB
    Alpha=LinearInterpolate(B=@n10, Alpha=@n11)
   )
   B=VectorParameter() {group=BottomBanner, parameterName=BorderColor, defaultValue=<rgba 1 0.8213 0 1>}.RGB
   Alpha=OneMinus(None=LinearInterpolate(A=@n22, B=@n38, Alpha=@n39))
  )
  ScalarParameter() {defaultValue=1, group=BottomBanner, parameterName=opacity, sliderMax=1, sortPriority=0}
 ]
}
